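# Translucent unlit Bacon Snack material for an alert-state icon: a question mark that can blend into an exclamation mark.
# QuestionProgress fills the question mark from the bottom via a UV sphere mask; QuestionColor alpha selects the exclamation.
# Colors, progress and timing come from custom primitive data, with a sine pulse on the inner color.
# World position offset billboards the mesh, scales and offsets it by camera distance, and grows it from its pivot via timestamped transition.
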Material /FNGameplayCues/Athena/Items/Consumables/BaconSnack/M_BaconSnack_AlertStates
graph {
 properties {
  bEnableResponsiveAA=true
  blendMode=BLEND_Translucent
  opacityMaskClipValue=0.5
  pixelDepthOffsetMode=PDOM_Legacy
  shadingModel=MSM_Unlit
  usedShadingModels=Unlit
 }
 nodes [
  n1 = TextureSample() {texture=&/T_BaconSnack_Question}
  n2 = Add(A=@n1.G, B=Multiply(A=@n1.B, B=ScalarParameter() {defaultValue=0.4, parameterName=InnerGlowScale}))
  n3 = ComponentMask(None=TextureCoordinate()) {g=true}
  n4 = Constant() {r=1}
  n5 = ScalarParameter() {bUseCustomPrimitiveData=true, parameterName=QuestionProgress, primitiveDataIndex=10}
  n6 = Constant() {r=1}
  n7 = Multiply(
   A=@n1.R
   B=LinearInterpolate(A=ScalarParameter() {defaultValue=0.15, parameterName=OpacityInner}, Alpha=SphereMask(A=@n3, B=@n4, Radius=@n5, Hardness=@n6))
  )
  n8 = VectorParameter() {bUseCustomPrimitiveData=true, parameterName=QuestionColor, defaultValue=<rgba 0.9792 0.409 0.0017 0>}
  LinearInterpolate(
   A=Saturate(None=Add(A=@n2, B=@n7))
   B=TextureSample() {texture=/Game/UI/Foundation/Textures/Icons/Replay/T-Icon-Replay-Exclamation-L}.A
   Alpha=@n8.A
  )
  n9 = MaterialFunctionCall!(
   WorldPosition
   PivotLocation
   WS Normals
   Custom Object Basis 1=Transform(None=Constant3Vector() {constant=<rgba 0 0 -1 1>}) {transformSourceType=TRANSFORMSOURCE_Local}
   Custom Object Basis 2=Transform(None=Constant3Vector() {constant=<rgba -1 0 0 1>}) {transformSourceType=TRANSFORMSOURCE_Local}
   Custom Object Basis 3=Transform(None=Constant3Vector() {constant=<rgba 0 1 0 1>}) {transformSourceType=TRANSFORMSOURCE_Local}
  ) {materialFunction=/Engine/Functions/Engine_MaterialFunctions02/WorldPositionOffset/AlignMeshToTheCamera}
  n10 = WorldPosition()
  n11 = Distance(
   A=MaterialFunctionCall() {materialFunction=&ObjectPivotPoint}.Object Pivot Location
   B=CameraPositionWS()
  )
  n12 = Multiply(
   A=Subtract(
    A=Add(A=@n9.World Position Offset, B=@n10.XYZ)
    B=MaterialFunctionCall() {materialFunction=&ObjectPivotPoint}.Object Pivot Location
   )
   B=Divide(A=@n11, B=ScalarParameter() {defaultValue=800, parameterName=ScaleByDistanceDivisor}) {constB=1000}
  )
  n13 = VectorParameter() {parameterName=BiasByDistanceDivisor, defaultValue=<rgba 0 0 0.02 0>}
  LinearInterpolate(
   A=Subtract(
    A=MaterialFunctionCall() {materialFunction=&ObjectPivotPoint}.Object Pivot Location
    B=WorldPosition().XYZ
   )
   B=Add(A=Add(A=@n9.World Position Offset, B=@n12), B=Multiply(A=@n11, B=@n13.RGB))
   Alpha=MaterialFunctionCall(
    TimestampA=ScalarParameter() {bUseCustomPrimitiveData=true, parameterName=TimestampA, primitiveDataIndex=4}
    TransitionLength_In=ScalarParameter() {bUseCustomPrimitiveData=true, parameterName=TransitionLength_In, primitiveDataIndex=5}
   ) {materialFunction=&MF_Timestamp_InOut_01}.SmoothStep
  )
  n14 = VectorParameter() {bUseCustomPrimitiveData=true, parameterName=QuestionColorInner, primitiveDataIndex=6, defaultValue=<rgba 0.9792 0.409 0.0017 0>}
  n15 = ConstantBiasScale(None=Sine(None=Time()) {period=0.5}) {scale=1.5}
  LinearInterpolate(A=LinearInterpolate(A=Multiply(A=@n14.RGB, B=@n15), B=@n8.RGB, Alpha=@n2), B=@n8.RGB, Alpha=@n8.RGB)
  Multiply(A=Time() {bOverride_Period=true, period=2}) {constB=0.5}
 ]
}
