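# Unlit translucent material for the BaconSnack NPC alert icon, showing a filling suspicion question mark or an exclamation mark.
# Custom primitive data drives state and fill, with offset texture samples adding drop shadows.
# A timestamped curve atlas bounces icon UVs through scale, shift and rotation, dampened during suspicion.
# Vertex offset math faces the camera, scales by distance and grows from the pivot; a DebugFloat4Values readout overlays color and opacity.
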Material /FNGameplayCues/Athena/Items/Consumables/BaconSnack/M_BaconSnack_AlertStates_v2
graph {
 properties {
  allowTranslucentCustomDepthWrites=true
  bEnableResponsiveAA=true
  blendMode=BLEND_Translucent
  numCustomizedUVs=1
  opacityMaskClipValue=0.5
  pixelDepthOffsetMode=PDOM_Legacy
  shadingModel=MSM_Unlit
  usedShadingModels=Unlit
 }
 nodes [
  ConstantBiasScale(None=Sine(None=Time()) {period=0.5}) {scale=50}
  Constant() {r=1}
  n1 = CurveAtlasRowParameter(CurveTime=MaterialFunctionCall(
    TimestampA=ScalarParameter() {bUseCustomPrimitiveData=true, parameterName=BounceStartTime, primitiveDataIndex=12}
    TransitionLength_In=ScalarParameter() {defaultValue=1.5, parameterName=BounceDuration}
   ) {materialFunction=&MF_Timestamp_InOut_01}.SmoothStep) {
   atlas=&/NPCAlertStatesCurves
   curve=&/NPCAlertBounceY
   parameterName=CurveParams
  }
  n2 = Constant() {r=1}
  n3 = If(
   A=ScalarParameter() {bUseCustomPrimitiveData=true, defaultValue=1, parameterName=IconState, primitiveDataIndex=3, sliderMax=1}
   B=@n2
   "A > B"=@n2
   "A == B"=@n2
   "A < B"=ScalarParameter() {defaultValue=0.25, parameterName=SuspicionAnimDampen}
  )
  n4 = Multiply(A=ConstantBiasScale(None=@n1.B) {bias=-1, scale=1}, B=@n3)
  n5 = Multiply(A=ConstantBiasScale(None=@n1.A) {bias=-1, scale=1}, B=@n3)
  n6 = VectorParameter() {parameterName="debugTexture Scale (V2)"}
  n7 = TextureCoordinate() {coordinateIndex=1}
  n8 = VectorParameter() {parameterName=debug, defaultValue=<rgba 0.3 0.3 0.7 0.7>}
  n9 = VectorParameter() {parameterName=debugspacing, defaultValue=<rgba 0 0.35 1 0.65>}
  n10 = MaterialFunctionCall!(
   Vector4=MaterialFunctionCall(X=@n1.R, Y=@n1.R, Z=@n4, A=@n5) {materialFunction=&MakeFloat4}.Result
   MaximumNumberOfDigits=Constant() {r=3}
   UVs=Add(
    A=AppendVector(A=@n6.B, B=@n6.B)
    B=MaterialFunctionCall(UVs=@n7, Texture Scale=@n6.RGB) {materialFunction=&ScaleUVsByCenter}.UVs
   )
   DebugTextLocation RG_UpperRight BA_LowerLeft=MaterialFunctionCall(R=@n8.R, G=@n8.R, B=@n8.R, A=@n8.R) {materialFunction=&AppendMany}.RGBA
   Component Spacing=MaterialFunctionCall(R=@n9.R, G=@n9.R, B=@n9.R, A=@n9.R) {materialFunction=&AppendMany}.RGBA
  ) {materialFunction=/Engine/Functions/Engine_MaterialFunctions02/Utility/DebugFloat4Values}
  n11 = MaterialFunctionCall(Float4=@n10.ColorCodedOutput) {materialFunction=&BreakOutFloat4Components}
  n12 = VectorParameter() {parameterName=ShadowColor, defaultValue=<rgba 0 0 0 0.8>}
  n13 = ScalarParameter() {bUseCustomPrimitiveData=true, parameterName=QuestionProgress, primitiveDataIndex=10, sliderMax=1}
  n14 = OneMinus(None=Clamp(None=@n13) {maxDefault=0.98})
  n15 = ScalarParameter() {parameterName=QuestionGlowExponent}
  n16 = TextureCoordinate()
  n17 = LinearInterpolate(
   A=VectorParameter() {parameterName=QuestionMarkColor, defaultValue=<rgba 1 1 1 0>}.RGB
   B=Multiply(A=VectorParameter() {parameterName=FillColor, defaultValue=<rgba 1 0.0715 0 1>}.RGB, B=Power(Base=@n14, Exp=@n15)) {desc=Pulse}
   Alpha=SphereMask(A=ComponentMask(None=@n16) {g=true}, B=Constant() {r=1}, Radius=@n13, Hardness=Constant() {r=1})
  )
  n18 = TextureSampleParameter2D() {
   parameterName=AlertShapes
   samplerSource=SSM_Clamp_WorldGroupSettings
   texture=&/T_NPC_AlertStates
  }
  n19 = ScalarParameter() {defaultValue=0.2, parameterName=ExclaimEdge}
  n20 = Add(A=@n19, B=ScalarParameter() {parameterName=EdgeSoftness})
  n21 = MaterialFunctionCall(Alpha=@n18.B, Min=@n19, Max=@n20) {materialFunction=&SmoothStep}
  n22 = VectorParameter() {parameterName=ShadowColor, defaultValue=<rgba 0 0 0 0.8>}
  n23 = VectorParameter() {parameterName=ExclamationColor, defaultValue=<rgba 1 0 0 1>}
  n24 = TextureSampleParameter2D() {
   parameterName=AlertShapes
   samplerSource=SSM_Clamp_WorldGroupSettings
   texture=&T_NPC_AlertStates
  }
  n25 = ScalarParameter() {defaultValue=0.2, parameterName=ExclaimEdge}
  n26 = Add(A=@n25, B=ScalarParameter() {parameterName=EdgeSoftness})
  n27 = MaterialFunctionCall(Alpha=@n24.R, Min=@n25, Max=@n26) {materialFunction=&SmoothStep}
  n28 = ScalarParameter() {bUseCustomPrimitiveData=true, defaultValue=1, parameterName=IconState, primitiveDataIndex=3, sliderMax=1}
  LinearInterpolate(
   A=LinearInterpolate(A=LinearInterpolate(A=@n12.RGB, B=@n17, Alpha=@n21.Result), B=LinearInterpolate(A=@n22.RGB, B=@n23.RGB, Alpha=@n27.Result), Alpha=@n28)
   B=MaterialFunctionCall(R=@n11.R, G=@n11.R, B=@n11.R) {materialFunction=&AppendMany}.RGB
   Alpha=@n10.GreyScaleOutput
  )
  n29 = VectorParameter() {parameterName=UVScale, defaultValue=<rgba 0.2 0.2 0 0>}
  n30 = AppendVector(A=@n29.R, B=@n29.R)
  n31 = AppendVector(A=@n1.R, B=@n1.R)
  n32 = TextureCoordinate()
  n33 = AppendVector(A=Constant(), B=@n5)
  MaterialFunctionCall(
   UVs=MaterialFunctionCall!(
    Texture Scale=Multiply(A=@n30, B=@n31)
    UVs=Add(A=@n32, B=@n33)
    Offset XY Texturesize Z=VectorParameter() {parameterName=UVOffset, defaultValue=<rgba 0.5 0.4 1 1>}.RGB
   ) {materialFunction=/Engine/Functions/Engine_MaterialFunctions01/Texturing/ScaleUVsAroundPoint}.UVs
   Rotation Center=AppendVector(A=Constant() {r=0.5}, B=OneMinus(None=@n5))
   "Rotation Angle (0-1)"=@n4
  ) {materialFunction=&CustomRotator}
  n34 = VectorParameter() {parameterName=ShadowOffset, defaultValue=<rgba -0.05 -0.08 0 0.41>}
  n35 = TextureSampleParameter2D(UVs=Add(A=TextureCoordinate(), B=AppendVector(A=@n34.R, B=@n34.R))) {
   parameterName=AlertShapes
   samplerSource=SSM_Clamp_WorldGroupSettings
   texture=&T_NPC_AlertStates
  }
  n36 = Max(
   A=Multiply(
    A=MaterialFunctionCall(Alpha=@n35.B, Min=@n19, Max=@n20) {materialFunction=&SmoothStep}.Result
    B=@n12.A
   )
   B=@n21.Result
  )
  n37 = VectorParameter() {parameterName=ShadowOffset, defaultValue=<rgba -0.05 -0.08 0 0.41>}
  n38 = TextureSampleParameter2D(UVs=Add(A=TextureCoordinate(), B=AppendVector(A=@n37.R, B=@n37.R))) {
   parameterName=AlertShapes
   samplerSource=SSM_Clamp_WorldGroupSettings
   texture=&T_NPC_AlertStates
  }
  n39 = Multiply(
   A=MaterialFunctionCall(Alpha=@n38.R, Min=@n25, Max=@n26) {materialFunction=&SmoothStep}.Result
   B=@n22.A
  )
  Max(A=LinearInterpolate(A=Saturate(None=@n36), B=Add(A=@n27.Result, B=@n39), Alpha=@n28), B=@n10.GreyScaleOutput)
  Multiply(A=Time() {bOverride_Period=true, period=2}) {constB=0.5}
  Multiply(A=Time() {bOverride_Period=true, period=5}, B=ScalarParameter() {defaultValue=1, parameterName=timespeed}) {constB=2, desc=test anim}
  n40 = MaterialFunctionCall() {materialFunction=&ObjectPivotPoint}
  n41 = WorldPosition()
  n42 = MaterialFunctionCall!(
   WorldPosition
   PivotLocation
   WS Normals
   Custom Object Basis 1=Transform(None=Constant3Vector() {constant=<rgba 0 0 -1 1>}) {transformSourceType=TRANSFORMSOURCE_Local}
   Custom Object Basis 2=Transform(None=Constant3Vector() {constant=<rgba -1 0 0 1>}) {transformSourceType=TRANSFORMSOURCE_Local}
   Custom Object Basis 3=Transform(None=Constant3Vector() {constant=<rgba 0 1 0 1>}) {transformSourceType=TRANSFORMSOURCE_Local}
  ) {materialFunction=/Engine/Functions/Engine_MaterialFunctions02/WorldPositionOffset/AlignMeshToTheCamera}
  n43 = Add(A=@n42.World Position Offset, B=WorldPosition().XYZ)
  n44 = MaterialFunctionCall() {materialFunction=&ObjectPivotPoint}
  n45 = Distance(
   A=MaterialFunctionCall() {materialFunction=&ObjectPivotPoint}.Object Pivot Location
   B=CameraPositionWS()
  )
  n46 = ScalarParameter() {defaultValue=800, parameterName=ScaleByDistanceDivisor}
  n47 = Add(A=@n42.World Position Offset, B=Multiply(A=Subtract(A=@n43, B=@n44.Object Pivot Location), B=Divide(A=@n45, B=@n46) {constB=1}))
  n48 = Multiply(A=@n45, B=VectorParameter() {parameterName=BiasByDistanceDivisor, defaultValue=<rgba 0 0 0.01 0>}.RGB)
  n49 = ScalarParameter() {bUseCustomPrimitiveData=true, parameterName=TimestampA, primitiveDataIndex=4}
  n50 = ScalarParameter() {bUseCustomPrimitiveData=true, parameterName=TransitionLength_In, primitiveDataIndex=5}
  Reroute(None=LinearInterpolate(
    A=Subtract(A=@n40.Object Pivot Location, B=@n41.XYZ)
    B=Add(A=@n47, B=@n48)
    Alpha=MaterialFunctionCall(TimestampA=@n49, TransitionLength_In=@n50) {materialFunction=&MF_Timestamp_InOut_01}.SmoothStep
   ))
  VectorParameter() {bUseCustomPrimitiveData=true, parameterName=QuestionColorInner, primitiveDataIndex=6, defaultValue=<rgba 0.9792 0.409 0.0017 0>}
  VectorParameter() {bUseCustomPrimitiveData=true, parameterName=QuestionColor, defaultValue=<rgba 0.9792 0.409 0.0017 0>}
 ]
}
